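# Unlit, translucent, two-sided mesh-particle material for a straight-blast smoke streak in the turret gun's impact effect.
# SplineThicken widens the spline mesh, with dynamic parameter Size blending the tip width.
# Two oppositely panning samples of a puffy smoke texture, optionally offset by particle random, combine with UV edge fades and particle alpha for opacity.
# A SphereMask moved along V by Wipe reveals or erases it, while particle color tints emissive.
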Material /Game/Gadgets/Assets/Turret_Gun/Particles/M_Impact_SmokeSpline_StraightBlast
graph {
 properties {
  bUsedWithMeshParticles=true
  bUsedWithNiagaraMeshParticles=true
  blendMode=BLEND_Translucent
  pixelDepthOffsetMode=PDOM_Legacy
  shadingModel=MSM_Unlit
  twoSided=true
  usedShadingModels=Unlit
 }
 nodes [
  MaterialFunctionCall(
   WidthBase=ScalarParameter() {defaultValue=50, parameterName="WidthBase (S)"}
   WidthTip=LinearInterpolate(
    A=ScalarParameter() {defaultValue=150, parameterName="WidthTip (S)_1"}
    B=ScalarParameter() {defaultValue=250, parameterName="WidthTip (S)"}
    Alpha=DynamicParameter() {paramNames [Size,Wipe,Param3,Param4]}.Size
   )
  ) {materialFunction=&SplineThicken}
  n1 = AppendVector(A=Constant(), B=ParticleRandom())
  n2 = Constant2Vector() {r=0.125}
  n3 = Add(
   A=TextureCoordinate() {uTiling=0.75, vTiling=0.15}
   B=StaticSwitchParameter(True=Add(A=@n1, B=@n2), False=@n2) {parameterName=Use Particle Random}
  )
  n4 = ScalarParameter() {defaultValue=0.25, parameterName="Rotation Angle (0-1) (S)"}
  n5 = TextureSample(UVs=Panner(Coordinate=MaterialFunctionCall(UVs=@n3, "Rotation Angle (0-1)"=@n4) {materialFunction=&CustomRotator}.Rotated Values) {speedX=-0.15}) {texture=/Game/Effects/Fort_Effects/Textures/Smoke/T_Puffy_Smoke_Trail_01}
  n6 = ParticleColor()
  MaterialFunctionCall(Emissive=Multiply(A=@n5.G, B=@n6.RGB)) {materialFunction=&MF_FX_EmissiveAdjustments}
  n7 = ComponentMask(None=TextureCoordinate() {vTiling=0.5}) {r=true}
  n8 = Multiply(A=Multiply(A=@n7, B=OneMinus(None=@n7)), B=ScalarParameter() {defaultValue=4, parameterName=B})
  n9 = ComponentMask(None=TextureCoordinate()) {g=true}
  n10 = Multiply(A=@n9, B=OneMinus(None=@n9))
  n11 = ScalarParameter() {defaultValue=1, parameterName=B_1}
  n12 = Add(A=TextureCoordinate() {uTiling=0.1, vTiling=0.5}, B=Constant2Vector() {g=0.15})
  n13 = ParticleRandom()
  n14 = Constant()
  n15 = StaticSwitchParameter(True=Add(A=@n12, B=AppendVector(A=@n13, B=@n14)), False=@n12) {parameterName=Use Particle Random}
  n16 = Multiply(
   A=Multiply(A=Power(Base=@n8), B=Multiply(A=@n10, B=@n11))
   B=TextureSample(UVs=Panner(Coordinate=@n15) {speedX=0.1}) {samplerSource=SSM_Wrap_WorldGroupSettings, texture=&T_Puffy_Smoke_Trail_01}.G
  )
  n17 = Constant()
  n18 = DynamicParameter() {paramNames [Size,Wipe,Param3,Param4]}
  Multiply(
   A=Multiply(A=Multiply(A=@n16, B=@n5.G), B=@n6.A)
   B=SphereMask(
    A=TextureCoordinate() {uTiling=0}
    B=AppendVector(A=@n17, B=@n18.Wipe)
    Radius=ScalarParameter() {defaultValue=0.5, parameterName=Radius}
    Hardness=ScalarParameter() {parameterName=Hardness}
   )
  )
 ]
}
