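# Additive mesh-particle material for turret gun overheat smoke, with panning textures over time.
# Vertex displacement combines V-gradient-weighted local offsets with panned distortion waves; both wave lerp inputs use WaveIntensity_Top, leaving WaveIntensity_Bot unused.
# Emissive blends time-of-day tint toward particle color along V over inky smoke, scaled by a noise mask and dynamic Glow.
# Opacity multiplies two noises, fresnel, distance and depth fades, V gradients and particle alpha.
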Material /Game/Gadgets/Assets/Turret_Gun/Particles/M_TurrentGun_OverHeatSmoke
graph {
 properties {
  bUsedWithMeshParticles=true
  bUsedWithNiagaraMeshParticles=true
  blendMode=BLEND_Additive
  pixelDepthOffsetMode=PDOM_Legacy
  usedShadingModels=Default Lit
 }
 nodes [
  n1 = Transform(None=VectorParameter() {parameterName=LocalPostioning}.RGB) {transformSourceType=TRANSFORMSOURCE_Local}
  n2 = TextureCoordinate()
  n3 = Power(
   Base=OneMinus(None=ComponentMask(None=@n2) {g=true})
   Exp=ScalarParameter() {defaultValue=2, parameterName=Edge Gradient Movement Control Exp}
  )
  n4 = Transform(None=VectorParameter() {parameterName=Initial WPO Offset, defaultValue=<rgba 50 5 -50 0>}.RGB) {transformSourceType=TRANSFORMSOURCE_Local}
  n5 = TextureCoordinate()
  n6 = Power(
   Base=OneMinus(None=ComponentMask(None=@n5) {g=true})
   Exp=ScalarParameter() {defaultValue=2, parameterName=Edge Gradient Movement Control Exp_1}
  )
  n7 = Add(A=VertexColor().R, B=VectorParameter() {parameterName=Distort_Panner_Tiler, defaultValue=<rgba 1 0.5 0 1>}.RGB)
  n8 = Multiply(A=Time(), B=ScalarParameter() {defaultValue=2, parameterName=Speed})
  n9 = Panner(Coordinate=Multiply(A=ComponentMask(None=@n7) {g=true, r=true}, B=TextureCoordinate()), Time=@n8) {speedY=0.25}
  n10 = TextureSample(UVs=@n9) {samplerType=SAMPLERTYPE_Normal, texture=&T_Weap_Distort_N_02}
  n11 = Multiply(A=MaterialFunctionCall(Float2=TextureCoordinate()) {materialFunction=&BreakOutFloat2Components}.G)
  n12 = TextureCoordinate() {vTiling=0.8}
  n13 = LinearInterpolate(
   B=AppendVector(
    A=ComponentMask(None=@n10.RGB) {g=true, r=true}
    B=TextureSample(UVs=@n9) {texture=&T_Smooth_tile}.R
   )
   Alpha=Multiply(
    A=OneMinus(None=@n11)
    B=MaterialFunctionCall(Float2=@n12) {materialFunction=&BreakOutFloat2Components}.G
   )
  )
  n14 = DynamicParameter() {defaultValue=<rgba 1 50 250 1>, paramNames [Glow, WaveIntensity_Bot, WaveIntensity_Top, Param4]}
  n15 = LinearInterpolate(A=@n14.WaveIntensity_Top, B=@n14.WaveIntensity_Top, Alpha=VertexColor().B)
  n16 = OneMinus(None=ComponentMask(None=TextureCoordinate()) {g=true})
  Add(A=Add(A=Multiply(A=@n1, B=@n3), B=Multiply(A=@n4, B=@n6)), B=Multiply(A=Multiply(A=@n13, B=@n15) {constB=10}, B=Clamp(None=@n16)))
  n17 = VectorParameter() {parameterName=Color_2, defaultValue=<rgba 1 0.6747 0.5123 1>}
  n18 = VectorParameter() {parameterName=Color_3, defaultValue=<rgba 1 1 1 1>}
  n19 = VectorParameter() {parameterName=Color_4, defaultValue=<rgba 0.6746 0.7841 1 1>}
  n20 = VectorParameter() {parameterName=Color_5, defaultValue=<rgba 0.4208 0.6361 1 1>}
  n21 = TextureCoordinate()
  n22 = ComponentMask(None=VectorParameter() {parameterName=MainSprayUVMulti, defaultValue=<rgba 2 1 0 0>}.RGB) {g=true, r=true}
  n23 = VectorParameter() {parameterName=MainSprayUVMulti_1}
  n24 = Panner(Coordinate=Add(A=Multiply(A=@n21, B=@n22), B=ComponentMask(None=@n23.RGB) {g=true, r=true}), Time=@n8) {speedY=0.25}
  n25 = Multiply(
   A=Power(Base=TextureSample(UVs=@n24) {texture=&T_Inky_Smoke_Tile}.G) {constExponent=0.8}
   B=ScalarParameter() {defaultValue=1, parameterName=MidDetail_Intensity}
  )
  n26 = MaterialFunctionCall(Float2=TextureCoordinate() {vTiling=0.8}) {materialFunction=&BreakOutFloat2Components}
  n27 = Multiply(A=@n26.G, B=@n26.G)
  n28 = LinearInterpolate(
   A=Multiply(
    A=MaterialFunctionCall(Color_Morning=@n17.RGB, Color_Day=@n18.RGB, Color_Evening=@n19.RGB, Color_Night=@n20.RGB) {materialFunction=&Color_Time_of_Day}.Result
    B=@n25
   )
   B=Multiply(A=ParticleColor().RGB, B=@n25)
   Alpha=Power(Base=Multiply(A=@n27, B=@n27)) {constExponent=4}
  )
  n29 = Multiply(
   A=TextureCoordinate() {uTiling=3}
   B=ComponentMask(None=VectorParameter() {parameterName=MainSprayUVMulti_2, defaultValue=<rgba 0.25 0.5 0 0>}.RGB) {g=true, r=true}
  )
  n30 = ComponentMask(None=VectorParameter() {parameterName=MainSprayUVMulti_3, defaultValue=<rgba 0.25 0.25 0 0>}.RGB) {g=true, r=true}
  n31 = TextureSample(UVs=Panner(Coordinate=Add(A=@n29, B=@n30), Time=@n8) {speedY=0.15}) {texture=/Game/Global/Textures/Default/Noise/T_FireNoise_02}
  n32 = Multiply(
   A=TextureCoordinate() {uTiling=2}
   B=ComponentMask(None=VectorParameter() {parameterName=MainSprayUVMulti_4, defaultValue=<rgba 2 1 0 0>}.RGB) {g=true, r=true}
  )
  n33 = ComponentMask(None=VectorParameter() {parameterName=MainSprayUVMulti_5}.RGB) {g=true, r=true}
  n34 = TextureSample(UVs=Panner(Coordinate=Add(A=@n32, B=@n33), Time=@n8) {speedY=0.25}) {texture=&T_FireNoise_02}
  n35 = Clamp(None=Power(Base=Multiply(A=@n31.G, B=@n34.G), Exp=ScalarParameter() {defaultValue=1, parameterName=Exp}) {constExponent=4})
  n36 = MaterialFunctionCall!(
   Axis=VectorParameter() {parameterName="Axis (V3)", defaultValue=<rgba 0 1 0 0>}.RGB
   Exponent=ScalarParameter() {defaultValue=1, parameterName=ExponentIn}
   Normal
  ) {materialFunction=/Game/Effects/Fort_Effects/Materials/_Functions/MF_AxisAlignedFrensnel}
  n37 = OneMinus(None=Fresnel(
    ExponentIn=ScalarParameter() {defaultValue=0.25, parameterName=ExponentIn_1}
    BaseReflectFractionIn=ScalarParameter() {defaultValue=0.04, parameterName=BaseReflectFractionIn}
   ))
  n38 = LinearInterpolate(B=Multiply(A=@n25, B=Multiply(A=@n36.Fresnel, B=@n37)), Alpha=ScalarParameter() {defaultValue=0.99, parameterName=Alpha}) {constA=1}
  n39 = ScalarParameter() {defaultValue=250, parameterName="Fade_Start_Distance (S)"}
  n40 = ScalarParameter() {defaultValue=125, parameterName="Fade_End_Distance (S)"}
  n41 = Multiply(
   A=Multiply(A=Multiply(A=@n35, B=@n38), B=ScalarParameter() {defaultValue=2, parameterName=B}) {constB=15}
   B=LinearInterpolate(Alpha=MaterialFunctionCall(Fade_Start_Distance=@n39, Fade_End_Distance=@n40) {materialFunction=&MF_Distance_SyncedFade_FX}.Result) {constA=0.25, constB=0.35}
  )
  n42 = LinearInterpolate(Alpha=@n41) {constA=0.25}
  MaterialFunctionCall(Emissive=LinearInterpolate(
    A=VectorParameter() {parameterName=NoGlow_Color, defaultValue=<rgba 0 0 0 1>}.RGB
    B=Multiply(A=@n28, B=@n42)
    Alpha=DynamicParameter() {defaultValue=<rgba 1 50 250 1>, paramNames [Glow, WaveIntensity_Bot, WaveIntensity_Top, Param4]}.Glow
   )) {materialFunction=&MF_FX_EmissiveAdjustments}
  n43 = ScalarParameter() {defaultValue=100, parameterName=FadeDistance}
  n44 = MaterialFunctionCall(Float2=TextureCoordinate() {vTiling=1.2}) {materialFunction=&BreakOutFloat2Components}
  n45 = Multiply(A=Multiply(A=@n41, B=DepthFade(FadeDistance=@n43)), B=Multiply(A=@n44.G, B=@n44.G))
  n46 = OneMinus(None=ComponentMask(None=TextureCoordinate()) {g=true})
  Multiply(A=ParticleColor().A, B=Clamp(None=Multiply(A=@n45, B=@n46)))
  Multiply(A=VectorParameter() {parameterName=Color, defaultValue=<rgba 1 1 1 1>}.RGB, B=@n25)
  ScalarParameter() {defaultValue=250, parameterName=WaveIntensity_Top}
  ScalarParameter() {defaultValue=50, parameterName=WaveIntensity_Bot}
  VertexColor()
 ]
}
