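# Translucent unlit particle material for the overheated barrel glow of the turret gun gadget.
# A dynamic OverHeatColor value, shaped by a power curve, blends particle-tinted low and high glow colors.
# Opacity combines particle alpha, an optional UV1 gradient mask, and axis-aligned or standard Fresnel, with ExpMultiplier scaling the standard exponent.
# An IsParticleLight static switch instead outputs intensity-scaled particle color and plain particle alpha.
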Material /Game/Gadgets/Assets/Turret_Gun/Particles/M_TurretGun_Overheated_Barrel
graph {
 properties {
  bUsedWithMeshParticles=true
  bUsedWithNiagaraMeshParticles=true
  bUsedWithParticleSprites=true
  blendMode=BLEND_Translucent
  pixelDepthOffsetMode=PDOM_Legacy
  shadingModel=MSM_Unlit
  usedShadingModels=Unlit
 }
 nodes [
  n1 = ParticleColor()
  n2 = Multiply(A=@n1.RGB, B=ScalarParameter() {defaultValue=1500, parameterName=LightIntensityColor})
  n3 = Clamp(None=Power(
    Base=DynamicParameter() {defaultValue=<rgba 0 1 1 1>, paramNames [OverHeatColor, ExpMultiplier, Param3, Param4]}.OverHeatColor
    Exp=ScalarParameter() {defaultValue=3, parameterName=ColorCurve_Power}
   ) {constExponent=8})
  n4 = ParticleColor()
  n5 = Multiply(A=VectorParameter() {parameterName=Glow_Color_Low, defaultValue=<rgba 1 0.2876 0 1>}.RGB, B=@n4.RGB)
  n6 = Multiply(A=ParticleColor().RGB, B=VectorParameter() {parameterName=Glow_Color_High, defaultValue=<rgba 15 0.15 0 1>}.RGB)
  MaterialFunctionCall(Emissive=StaticSwitchParameter(True=LinearInterpolate(A=@n1.RGB, B=@n2, Alpha=@n3), False=LinearInterpolate(A=@n5, B=@n6, Alpha=@n3)) {parameterName=IsParticleLight?}) {materialFunction=&MF_FX_EmissiveAdjustments}
  ScalarParameter() {parameterName="LERP "}
  n7 = Constant() {r=1}
  n8 = ComponentMask(None=TextureCoordinate() {coordinateIndex=1}) {g=true}
  n9 = LinearInterpolate(A=@n8, B=OneMinus(None=@n8), Alpha=ScalarParameter() {parameterName=Invert Gradient})
  n10 = LinearInterpolate(
   A=@n9
   B=Power(Base=@n9, Exp=ScalarParameter() {defaultValue=5, parameterName=Gradient Power})
   Alpha=ScalarParameter() {parameterName=Use Gradient Power}
  )
  n11 = ScalarParameter() {defaultValue=1, parameterName=UseGradientMask, sliderMax=1}
  n12 = ScalarParameter() {defaultValue=0.25, parameterName="Exponent (S)"}
  n13 = MaterialFunctionCall!(Axis=VectorParameter() {parameterName="Axis (V3)", defaultValue=<rgba 0 1 0 0>}.RGB, Exponent=@n12, Normal) {materialFunction=/Game/Effects/Fort_Effects/Materials/_Functions/MF_AxisAlignedFrensnel}
  n14 = ScalarParameter() {parameterName=InvertFresnel}
  n15 = OneMinus(None=ScalarParameter() {defaultValue=1, parameterName=Fresnel On OFF})
  n16 = Add(A=LinearInterpolate(A=OneMinus(None=@n13.Fresnel), B=@n13.Fresnel, Alpha=@n14), B=@n15)
  n17 = MaterialFunctionCall(Exponent=Multiply(A=@n12, B=DynamicParameter() {defaultValue=<rgba 0 1 1 1>, paramNames [OverHeatColor, ExpMultiplier, Param3, Param4]}.ExpMultiplier)) {materialFunction=&MF_Fresnel_FX}
  n18 = Add(A=LinearInterpolate(A=@n17.Fresnel Inverted, B=@n17.Fresnel Inverted, Alpha=@n14), B=@n15)
  n19 = Multiply(
   A=Multiply(A=@n4.A, B=LinearInterpolate(A=@n7, B=@n10, Alpha=@n11))
   B=LinearInterpolate(
    A=Clamp(None=@n16)
    B=Clamp(None=@n18)
    Alpha=ScalarParameter() {desc="1 = Standard Fresnel\r\n0 = Axis Aligned Fresnel", parameterName=Use Standard Fresnel}
   )
  )
  n20 = ParticleColor()
  StaticSwitchParameter(
   True=ParticleColor().A
   False=LinearInterpolate(A=@n19, B=Multiply(A=@n19, B=@n20.A), Alpha=ScalarParameter() {parameterName=Use Particle Alpha})
  ) {parameterName=IsParticleLight?}
 ]
}
