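# Unlit translucent Niagara mesh-particle material for the glowing overheated barrel of the turret gun.
# A red-orange high-intensity Glow_Color_High parameter feeds the MF_FX_EmissiveAdjustments function for the glow color.
# A separate mask raises vertex color red to a power, scales it by the OverHeatAmount dynamic parameter, then multiplies by a Fresnel term, so particles control the heat strength.
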Material /Game/Gadgets/Assets/Turret_Gun/Particles/M_TurretGun_Overheated_Barrel_VertexColor
graph {
 properties {
  bUsedWithNiagaraMeshParticles=true
  blendMode=BLEND_Translucent
  pixelDepthOffsetMode=PDOM_Legacy
  shadingModel=MSM_Unlit
  usedShadingModels=Unlit
 }
 nodes [
  MaterialFunctionCall(Emissive=VectorParameter() {parameterName=Glow_Color_High, defaultValue=<rgba 15 0.15 0 1>}.RGB) {materialFunction=&MF_FX_EmissiveAdjustments}
  n1 = VertexColor()
  Multiply(
   A=Multiply(
    A=Power(Base=@n1.R)
    B=DynamicParameter() {defaultValue=<rgba 0 1 1 1>, paramNames [OverHeatAmount, ExpMultiplier, Param3, Param4]}.OverHeatAmount
   )
   B=Fresnel() {baseReflectFraction=0.1}
  )
 ]
}
