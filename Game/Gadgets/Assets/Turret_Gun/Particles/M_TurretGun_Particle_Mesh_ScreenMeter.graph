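# Unlit, two-sided translucent material for the turret gun's screen meter particle mesh, drawing an overheat bar.
# A U-axis sphere mask fills the bar, with radius blended between OverHeatAmount and the OverHeatColor dynamic parameter via UseParticleDP.
# LeftOrRight mirrors both fill origin and the gradient between two overheat colors.
# Masked color feeds MF_FX_EmissiveAdjustments; Blink is disabled by a constant zero blend, leaving RedBlink unused.
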Material /Game/Gadgets/Assets/Turret_Gun/Particles/M_TurretGun_Particle_Mesh_ScreenMeter
graph {
 properties {
  bUsedWithMeshDeformer=true
  bUsedWithMeshParticles=true
  bUsedWithNiagaraMeshParticles=true
  bUsedWithSkeletalMesh=true
  blendMode=BLEND_Translucent
  pixelDepthOffsetMode=PDOM_Legacy
  shadingModel=MSM_Unlit
  twoSided=true
  usedShadingModels=Unlit
 }
 nodes [
  n1 = VectorParameter() {parameterName=OverHeat_Color, defaultValue=<rgba 1 0 0 1>}
  n2 = VectorParameter() {parameterName=OverHeat_Color_Low, defaultValue=<rgba 1 0.7435 0 1>}
  n3 = ComponentMask(None=TextureCoordinate()) {r=true}
  n4 = OneMinus(None=@n3)
  n5 = DynamicParameter() {defaultValue=<rgba 0 1 1 1>, paramNames [OverHeatColor, LeftOrRight, RedBlink, Param4]}
  n6 = Clamp(None=Power(Base=LinearInterpolate(A=@n3, B=@n4, Alpha=@n5.LeftOrRight)))
  n7 = Constant() {r=1}
  n8 = Constant()
  n9 = DynamicParameter() {defaultValue=<rgba 0 1 1 1>, paramNames [OverHeatColor, LeftOrRight, RedBlink, Param4]}
  n10 = SphereMask(
   A=TextureCoordinate() {vTiling=0}
   B=AppendVector(A=LinearInterpolate(A=@n7, B=@n8, Alpha=@n9.LeftOrRight), B=Constant())
   Radius=LinearInterpolate(
    A=ScalarParameter() {parameterName=OverHeatAmount}
    B=DynamicParameter() {defaultValue=<rgba 0 1 1 1>, paramNames [OverHeatColor, LeftOrRight, RedBlink, Param4]}.OverHeatColor
    Alpha=ScalarParameter() {defaultValue=1, parameterName=UseParticleDP}
   )
   Hardness=ScalarParameter() {defaultValue=1, parameterName=Hardness}
  )
  n11 = VectorParameter() {parameterName=Blink, defaultValue=<rgba 1 0 0 1>}
  n12 = Constant()
  MaterialFunctionCall(Emissive=Multiply(A=LinearInterpolate(A=@n1.RGB, B=@n2.RGB, Alpha=@n6), B=LinearInterpolate(A=@n10, B=@n11.RGB, Alpha=@n12))) {materialFunction=&MF_FX_EmissiveAdjustments}
  Reroute(None=@n10)
 ]
}
